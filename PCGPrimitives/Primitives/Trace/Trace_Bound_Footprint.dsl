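; PCG primitive graph Trace_Bound_Footprint, which by its name traces the footprint of input bounds on a selectable planar surface.
; It tags the primary input "Input:0", then chains template input management, a core process subgraph and output preparation.
; The core subgraph receives the plane, position offset, scale and as-point options and holds the actual footprint logic.
; Graph parameters make input processing, fallback, core and output stages swappable.
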(graph Trace_Bound_Footprint
  (params (InputsProcessingGraph :type (SoftObject /Script/PCG.PCGGraphInterface) :default /PCGPrimitives/Primitives/_Template_/Primitives_InputsProcessing.Primitives_InputsProcessing) (InputsFallbackGraph :type (SoftObject /Script/PCG.PCGGraphInterface) :default /PCGPrimitives/Primitives/_Template_/Primitives_InputsFallback.Primitives_InputsFallback) (CoreProcessGraph :type (SoftObject /Script/PCG.PCGGraphInterface) :default /PCGPrimitives/Primitives/Trace/Subgraphs/Trace_Bound_Footprint_CoreProcess.Trace_Bound_Footprint_CoreProcess) (PrepareOutputsGraph :type (SoftObject /Script/PCG.PCGGraphInterface) :default /PCGPrimitives/Primitives/_Template_/Primitives_PrepareOutputs.Primitives_PrepareOutputs) (PlanarSurface :type (Enum /PCGPrimitives/Primitives/Trace/Resources/EPCG_PlanarSurfacesEnum.EPCG_PlanarSurfacesEnum) :default NewEnumerator10) (PositionOffset :type Vector) (ScaleFactor :type Vector :default (1.0 1.0 1.0)) (AsPoint :type Bool))
  (inputs (PrimaryInput :type Any))
  (bind AddTags_1 (Add%20Tags :TagsToAdd "Input:0"))
  ;[0 0 0 0.5 32] Inputs
  (bind Subgraph_0 (Subgraph :Graph /PCGPrimitives/Primitives/_Template_/Subgraphs/Primitives_InputsManagement.Primitives_InputsManagement :InputsFallbackGraph InputsFallbackGraph :InputsProcessingGraph InputsProcessingGraph))
  ;[0 0 0 0.5 32] Core Process
  (bind Subgraph_3 (Subgraph :Graph /PCGPrimitives/Primitives/Trace/Subgraphs/Trace_Bound_Footprint_CoreProcess.Trace_Bound_Footprint_CoreProcess :AsPoint AsPoint :PlanarSurface PlanarSurface :PositionOffset PositionOffset :ScaleFactor ScaleFactor :SubgraphOverride-param CoreProcessGraph))
  ;[0 0 0 0.5 32] Prepare Output
  (bind Subgraph_4 (Subgraph :Graph /PCGPrimitives/Primitives/_Template_/Primitives_PrepareOutputs.Primitives_PrepareOutputs :SubgraphOverride-param PrepareOutputsGraph))
  (connect DefaultInputNode PrimaryInput AddTags_1 In)
  (connect AddTags_1 Out Subgraph_0 In)
  (connect Subgraph_0 Out Subgraph_3 In)
  (connect Subgraph_3 Out Subgraph_4 In)
  (output Out Subgraph_4))
