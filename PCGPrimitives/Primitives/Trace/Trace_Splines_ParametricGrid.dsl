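; PCG primitive graph that traces a parametric spline grid from two inputs, returning Cross Splines and Span Splines.
; Inputs are tagged by slot, gathered, then passed through input management, a core process subgraph taking transition count and linear interpolation, and output preparation.
; Graph parameters expose the input processing, fallback, core and output stages as swappable subgraphs.
; A "Longitude" tag filter splits results: matches become Cross Splines, the rest Span Splines.
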(graph Trace_Splines_ParametricGrid
  (params (InputsProcessingGraph :type (SoftObject /Script/PCG.PCGGraphInterface) :default /PCGPrimitives/Primitives/_Template_/Primitives_InputsProcessing.Primitives_InputsProcessing) (InputsFallbackGraph :type (SoftObject /Script/PCG.PCGGraphInterface) :default /PCGPrimitives/Primitives/_Template_/Primitives_InputsFallback.Primitives_InputsFallback) (CoreProcessGraph :type (SoftObject /Script/PCG.PCGGraphInterface) :default /PCGPrimitives/Primitives/Trace/Subgraphs/Trace_Splines_ParametricGrid_CorePocess.Trace_Splines_ParametricGrid_CorePocess) (PrepareOutputsGraph :type (SoftObject /Script/PCG.PCGGraphInterface) :default /PCGPrimitives/Primitives/Shared/PrepareOutputs/PrepareOutputs_ClosedSplinesAndSplines.PrepareOutputs_ClosedSplinesAndSplines) (TransitionSplines :type Int :default 5) (LinearInterpolation :type Bool))
  (inputs (PrimaryInput :type Any) (SecondaryInput :type Any))
  (outputs (Cross%20Splines :type Any) (Span%20Splines :type Any))
  (bind AddTags_1 (Add%20Tags :TagsToAdd "Input:0"))
  (bind AddTags_2 (Add%20Tags :TagsToAdd "Input:1"))
  (bind Gather_0 (Gather))
  (bind FilterDataByTag_0 (Filter%20Data%20By%20Tag :SelectedTags "Longitude"))
  ;[0 0 0 0.5 32] Inputs
  (bind Subgraph_0 (Subgraph :Graph /PCGPrimitives/Primitives/_Template_/Subgraphs/Primitives_InputsManagement.Primitives_InputsManagement :InputsFallbackGraph InputsFallbackGraph :InputsProcessingGraph InputsProcessingGraph))
  ;[0 0 0 0.5 32] Core Process
  (bind Subgraph_3 (Subgraph :Graph /PCGPrimitives/Primitives/Trace/Subgraphs/Trace_Splines_ParametricGrid_CorePocess.Trace_Splines_ParametricGrid_CorePocess :Linear_Interpolation LinearInterpolation :SubgraphOverride-param CoreProcessGraph :Transition_Splines TransitionSplines))
  ;[0 0 0 0.5 32] Prepare Output
  (bind Subgraph_4 (Subgraph :Graph /PCGPrimitives/Primitives/_Template_/Primitives_PrepareOutputs.Primitives_PrepareOutputs :SubgraphOverride-param PrepareOutputsGraph))
  (connect DefaultInputNode PrimaryInput AddTags_1 In)
  (connect DefaultInputNode SecondaryInput AddTags_2 In)
  (connect Subgraph_4 Out FilterDataByTag_0 In)
  (connect AddTags_1 Out Gather_0 In)
  (connect AddTags_2 Out Gather_0 In2)
  (connect Gather_0 Out Subgraph_0 In)
  (connect Subgraph_0 Out Subgraph_3 In)
  (connect Subgraph_3 Out Subgraph_4 In)
  (output Cross%20Splines FilterDataByTag_0.InsideFilter)
  (output Span%20Splines FilterDataByTag_0.OutsideFilter))
